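# Editor utility widget that finds and selects building actors with negative Z scale in the level.
# Its Find Scaled Actors button gets all BuildingActor instances and checks each one's transform scale, testing only the Z axis.
# Matches are added to AR_ValidActors, and the editor selection is set to that array after each addition.
# The array is never cleared between clicks.
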EditorUtilityWidgetBlueprint /Game/ArtTools/Environment/NegativeActorScale/EUW_NegativeActorScale
graphs [EventGraph {nodes [
   "On Clicked (BTN_GetScaledActors)"!() -> (OutputDelegate, then, widget:User Widget Object Reference, event:Pointer Event Structure) {is=title K2Node_ComponentBoundEvent}
   > n1 = GetAllActorsOfClass!(execute, WorldContextObject:Object Reference, ActorClass:Actor Class Reference=/Script/FortniteGame.BuildingActor) -> (then, OutActors:Array of Building Actor Object References)
   > n2 = For Each Loop(Array:Array of Building Actor Object References=@n1) -> (Array Element:Building Actor Object Reference)
   n3 = BreakTransform(InTransform=Get Actor Transform(self=@n2))
   > Branch(Condition="float < float"(A=BreakVector(InVec=@n3.Scale).Z))
   > Add(
    TargetArray:Array of Building Actor Object References=Get AR_ValidActors() -> (AR_ValidActors:Array of Building Actor Object References)
    NewItem:"Building Actor Object Reference (by ref)"=@n2
   ) {is=title K2Node_CallArrayFunction}
   > SetSelectedLevelActors(self=Get EditorActorSubsystem(), ActorsToSelect=Get AR_ValidActors())
   Event PreConstruct()
   Event Construct()
   Event Tick()
  ]}]
variables [AR_ValidActors {type="array<BuildingActor>", default []}]
widgets {
 info {rootWidgetClass=&VerticalBox, widgetCount=2}
 widgets [
  VerticalBox_0 {widgetClassPath=&VerticalBox, widget=VerticalBox_0}
  BTN_GetScaledActors {
   widgetClassPath=&TA_Generic01_Button
   parent=VerticalBox_0
   widget=BTN_GetScaledActors
   properties {label=Find Scaled Actors}
   slot {class=VerticalBoxSlot, properties {padding=<ltrb 5 5 5 5>}}
  }
 ]
}
